(event EventBeginPlay
  (CallFunction|SetDecalTransform))

(fn SetDecalTransform
  (bind _returnvalue (Transformation|GetActorScale3D))
  (bind (_returnvalue_x _returnvalue_y) (Math|Vector|BreakVector (Transformation|GetActorLocation)))
  (bind (_returnvalue_x_2 _returnvalue_y_2) (Math|Vector|BreakVector (Transformation|GetActorScale3D)))
  (bind (_returnvalue_x_3 _returnvalue_y_3) (Math|Vector|BreakVector (* _returnvalue "100.000000,100.000000,100.000000")))
  (Transformation|SetActorScale3D 0 (Math|Vector|MakeVector _returnvalue_x_2 _returnvalue_y_2 (Math|Float|Min(Float) _returnvalue_x_2 _returnvalue_y_2)))
  (Rendering|Material|SetVectorParameterValue "/Game/Packages/Fortress_SharedMaterials/GlobalMaterialParameters/FortniteMaterialParameters.FortniteMaterialParameters" "BiomeDecal" (Math|Color|MakeColor _returnvalue_x _returnvalue_y _returnvalue_x_3 _returnvalue_y_3) :!class /Script/Engine.KismetMaterialLibrary)
  (Development|PrintString (Utilities|String|ToString(Float) _returnvalue_x) true true "(R=0.000000,G=0.660000,B=1.000000,A=1.000000)" 1e+01)
  (Development|PrintString (Utilities|String|ToString(Float) _returnvalue_y) true true "(R=0.000000,G=0.660000,B=1.000000,A=1.000000)" 1e+01)
  (Development|PrintString (Utilities|String|ToString(Float) _returnvalue_x_3) true true "(R=1.000000,G=0.000000,B=0.422856,A=1.000000)" 1e+01)
  (Development|PrintString (Utilities|String|ToString(Float) _returnvalue_y_3) true true "(R=1.000000,G=0.000000,B=0.422856,A=1.000000)" 1e+01))

(fn UpdateLandscape
  (bind (_returnvalue_x _returnvalue_y) (Math|Vector|BreakVector (Transformation|GetActorScale3D)))
  (bind (_returnvalue_x_2 _returnvalue_y_2) (Math|Vector|BreakVector (Transformation|GetActorLocation)))
  (bind _returnvalue (/ 3.85 1e+04))
  (bind _returnvalue_2 (Math|Vector2D|MakeVector2D _returnvalue _returnvalue))
  (bind (_returnvalue_x_3 _returnvalue_y_3 _returnvalue_z) (Math|Vector|BreakVector (Transformation|GetActorScale3D)))
  (bind (_returnvalue_x_4 _returnvalue_y_4 _returnvalue_z_2) (Math|Vector|BreakVector (Transformation|GetActorScale3D)))
  (bind (_returnvalue_x_5 _returnvalue_y_5) (Math|Vector|BreakVector (Transformation|GetActorScale3D)))
  (Utilities|FlowControl|Sequence
    (:then_0
      ; Use the RVT Volume to determine UV values. UV scale and UV offset.
      (LocalVariables|SetRVTMatchFound false)
      (bind _outactors (Actor|GetAllActorsOfClass "/Script/Engine.RuntimeVirtualTextureVolume"))
      (for _array_element _outactors
        (if (== (Variables|/Script/Engine.RuntimeVirtualTextureComponent|GetStreamingTexture (Variables|/Script/Engine.RuntimeVirtualTextureVolume|GetVirtualTextureComponent _array_element)) (Variables|BiomeDecal|GetLandscapeHeightMap))
          (LocalVariables|SetRVTOrigin (Transformation|GetActorLocation _array_element))
          (LocalVariables|SetRVTScale (Transformation|GetActorScale3D _array_element))
          (LocalVariables|SetRVTMatchFound true))))
    (:then_1
      (if (and (LocalVariables|GetRVTMatchFound) (not (Variables|BiomeDecal|GetLegacyUVs)))
        (LocalVariables|SetUVScale (Math|Vector2D|MakeVector2D (.x (/ (* (Transformation|GetActorScale3D) 1e+02) (LocalVariables|GetRVTScale))) (.y (/ (* (Transformation|GetActorScale3D) 1e+02) (LocalVariables|GetRVTScale)))))
        (LocalVariables|SetUVOffset (Math|Vector2D|MakeVector2D (.x (/ (- (- (Transformation|GetActorLocation) (* (Transformation|GetActorScale3D) 5e+01)) (LocalVariables|GetRVTOrigin)) (LocalVariables|GetRVTScale))) (.y (/ (- (- (Transformation|GetActorLocation) (* (Transformation|GetActorScale3D) 5e+01)) (LocalVariables|GetRVTOrigin)) (LocalVariables|GetRVTScale)))))
        (LocalVariables|SetDisplacementMagnitude (/ (.z (LocalVariables|GetRVTScale)) _returnvalue_z_2))
        (else
          ; LEGACY UV SCALE AND OFFSET AND MAGNITUDE
          (LocalVariables|SetUVScale (* (Math|Vector2D|MakeVector2D _returnvalue_x _returnvalue_y) _returnvalue_2))
          (LocalVariables|SetUVOffset (* (- "(X=1.000000,Y=1.000000)" (* _returnvalue_2 (+ (* (Math|Vector2D|MakeVector2D _returnvalue_x_2 _returnvalue_y_2) "(X=-0.020000,Y=-0.020000)") (Math|Vector2D|MakeVector2D _returnvalue_x_5 _returnvalue_y_5)))) "(X=0.500000,Y=0.500000)"))
          (LocalVariables|SetDisplacementMagnitude (* (/ 5.4e+02 _returnvalue_z) 1e+02)))))
    (:then_2
      (bind _returnvalue_3 (AddComponent|NODE_AddDynamicMeshComponent-0))
      (bind _output_get (Variables|Default|SetDynamicMeshComponent _returnvalue_3))
      (bind _returnvalue_4 (DynamicMeshComponent|GetDynamicMesh _output_get))
      (bind _returnvalue_5 (GeometryScript|Primitives|AppendRectangleXY :TargetMesh _returnvalue_4 :DimensionX 1e+02 :DimensionY 1e+02 :StepsWidth (Variables|BiomeDecal|GetLandscapeSubdivisions) :StepsHeight (Variables|BiomeDecal|GetLandscapeSubdivisions)))
      (bind _returnvalue_6 (GeometryScript|Deformations|ApplyDisplacefromTextureMap :TargetMesh _returnvalue_5 :Texture (Variables|/Script/Engine.VirtualTextureBuilder|GetTexture (Variables|BiomeDecal|GetLandscapeHeightMap)) :Options (Utilities|Struct|MakeGeometryScriptDisplacefromTextureOptions (* (LocalVariables|GetDisplacementMagnitude) (Variables|BiomeDecal|GetLandscapeHeightMultiplier)) (LocalVariables|GetUVScale) (LocalVariables|GetUVOffset) (Math|Float|SelectFloat 0.2 0.0 (Variables|BiomeDecal|GetLegacyUVs))) :UVLayer 0))
      (bind _returnvalue_7 (GeometryScript|Normals|RecomputeNormals _returnvalue_6))
      (bind _returnvalue_8 (GeometryScript|StaticMesh|CopyMeshtoStaticMesh _returnvalue_7 (Variables|Default|GetTo%20Static%20Mesh%20Asset))
        (:Failure
          _)
        (:Success
          _))
      (DynamicMesh|Reset _returnvalue_8)
      (Components|DestroyComponent (Variables|Default|GetDynamicMeshComponent)))
    (:then_3
      ; Set the Z position to match landscape
      (if (and (Variables|BiomeDecal|GetSetZ) (not (Variables|BiomeDecal|GetLegacyUVs)))
        (Transformation|SetActorLocation (Math|Vector|MakeVectorNetQuantize (.x (Transformation|GetActorLocation)) (.y (Transformation|GetActorLocation)) (.z (LocalVariables|GetRVTOrigin))))))))

(fn UpdateTexture
  (TextureGraph|ExportTextureGraph (Variables|BiomeDecal|GetTextureProcessor) true true true))

(fn ConstructionScript
  (bind (_min _max) (Components|StaticMesh|GetLocalBounds (Variables|StaticMeshActor|GetStaticMeshComponent)))
  (bind (_returnvalue_location _returnvalue_rotation _returnvalue_scale) (Math|Transform|BreakTransform (Transformation|GetActorTransform)))
  (CallFunction|SetDecalTransform)
  (bind _returnvalue (Actor|GetActorOfClass "/Game/Athena/Helios/Environments/Landscape/WorldBuildingTools/PCGGrass/BP_MeshGrass_GlobalControls.BP_MeshGrass_GlobalControls_C"))
  (Utilities|IsValid _returnvalue
    (:Is%20Valid
      (Class|BPIMeshGrass|AddWorldTexture _returnvalue "BiomeDecal_01" (Utilities|Struct|MakeMeshGrassWorldTextureData "/Hermes/Terrain/BiomeDecal/RT_BiomeDecal_Tex.RT_BiomeDecal_Tex" "Red" 0.2 true (Math|Transform|MakeTransform _returnvalue_location _returnvalue_rotation (* _returnvalue_scale _max)))))))
